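# Niagara ribbon trail system for the York glider, drawing two short-lived pale cyan anim trails between bone or socket pairs.
# A single emitter spawns both trails, with User.bIsFullyDeployed switching the spawn interval between 1 and 0, and feeds two ribbon renderers.
# Alpha fades over life and scales with clamped User.ForwardDot, while ribbon width tapers by curve.
# A camera offset is applied, and the inherit-velocity module is disabled.
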NiagaraSystem /BRCosmetics/Gadgets/Assets/VinderTech_GliderChute/Glider_York_Male/FX/NS_Glider_York_Trail
dependencies {
 usedDataInterfaces [&NiagaraDataInterfaceCurve]
 usedDynamicInputs []
 usedModules [/Niagara/Modules/System/SystemLifeCycle, &SystemState]
}
emitters {FN_AnimTrail_V4001 {
  emitterUpdateScript [
   &EmitterState()
   &FN_SpawnAnimTrail_v2(
    Number of Anim Trails=2
    Spawn Interval:NiagaraInt32=&MakeCustomIntFromBool(Bool=$User.bIsFullyDeployed, True INT=1, False INT=0)
   )
  ]
  particleSpawnScript [
   /Game/Effects/Niagara/Modules/AnimTrail/FN_InitializeAnimTrail/FN_InitializeAnimTrailBoneIndecies(Second Bone or Socket Index=1)
   &FN_InitializeAnimTrailBoneIndecies(AnimTrail Selection=1, First Bone or Socket Index=2, Second Bone or Socket Index=3) {module=FN_InitializeAnimTrailBoneIndecies001}
   /Game/Effects/Niagara/Modules/AnimTrail/FN_InitializeAnimTrail/FN_InitializeAnimTrail(AnimTrailWidth=0.6, Color=<rgba 0.7966 1.045 1.1 1>, Lifetime=0.25)
  ]
  particleUpdateScript [
   &InheritVelocity(Module.Limit Speed=true) {enabled=false}
   &SolveForcesAndVelocity(Clamp Velocity=false, Limit Acceleration=false)
   /Game/Effects/Niagara/Modules/AnimTrail/FN_UpdateAnimTrail/FN_UpdateAnimTrail()
   &ScaleColor(
    Scale Alpha=&Multiply_Float(
     A=&FloatFromCurve(FloatCurve=DataInterface() {propertyValues {Curve {keys [<at 0 1>,<at 0.2538 1>,<at 1 0>]}}}, Scale Curve=0.3)
     B=&ClampFloat(Float=$User.ForwardDot)
    )
    Scale RGB=&MakeVectorFromLinearColorRGB(Color=<rgba 0.6737 0.9092 1 1>)
    ScaleA=true
    ScaleRGB=true
   )
   &ScaleRibbonWidth(Ribbon Width Scale=&FloatFromCurve(FloatCurve=DataInterface() {propertyValues {Curve {keys [<at 0 1>,<at 0.9847 0.159>]}}}))
   &CameraOffset(Camera Offset Amount=20)
  ]
  rendererClasses [&NiagaraRibbonRendererProperties]
  renderers [
   {rendererClass=&NiagaraRibbonRendererProperties}
   {rendererClass=&NiagaraRibbonRendererProperties, rendererIndex=1}
  ]
 }}
summary {
 systemName=NS_Glider_York_Trail
 emitters [{emitterName=FN_AnimTrail_V4001, rendererClasses [&NiagaraRibbonRendererProperties]}]
 userVariables [
  User.bIsFullyDeployed {type=NiagaraBool}
  User.ForwardDot {type=NiagaraFloat, defaultValue=0, description=""}
  User.LocalAlpha {type=NiagaraFloat, defaultValue=0, description=""}
  User.PlayerSpeed {type=NiagaraFloat, defaultValue=0, description=""}
  User.RibbonIDOffset {type=NiagaraInt32}
  User.RightDot {type=NiagaraFloat, defaultValue=0, description=""}
  User.SkeletalMesh {
   type=NiagaraDataInterfaceSkeletalMesh
   defaultValue {
    struct=NiagaraExt_VariableValue_DataInterface
    value {dataInterface=&self:NiagaraDataInterfaceSkeletalMesh_0}
   }
  }
 ]
}
