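# Master material for the York male glider's sirens, adding animated glowing lights to skeletal meshes between the character PreFX and PostFX functions.
# A time-driven sine pulse, offset along UV V and scaled by SirenSpeed, modulates emissive and light size.
# The rayTraceSphereSirens switch selects four ray-traced spheres spaced along the SocketLocationL/R line, or four UV sphere masks with view-dependent horizontal shift faking parallax.
# Powered masks build tinted glows with white-hot cores.
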Material /BRCosmetics/Gadgets/Assets/VinderTech_GliderChute/Glider_York_Male/Materials/M_Glider_York_Male_Sirens_MASTER
graph {
 properties {
  bUseMaterialAttributes=true
  bUsedWithMeshDeformer=true
  bUsedWithSkeletalMesh=true
  pixelDepthOffsetMode=PDOM_Legacy
  usedShadingModels=Default Lit
 }
 nodes [
  n1 = VectorParameter() {parameterName=SocketLocationL, defaultValue=<rgba 2760 1336 2075 0>}
  Add(A=PixelNormalWS(), B=@n1.RGB)
  Constant3Vector()
  n2 = CameraVectorWS()
  n3 = Transform(None=Constant3Vector() {constant=<rgba 0 -1 0 0>}) {transformSourceType=TRANSFORMSOURCE_Local}
  n4 = Multiply(A=Add(A=DotProduct(A=@n2, B=@n3))) {constB=0.5}
  LinearInterpolate(A=Constant3Vector() {constant=<rgba 1 0 0 0>}, B=Constant3Vector() {constant=<rgba 0 1 0 0>}, Alpha=@n4)
  n5 = MaterialFunctionCall() {materialFunction=&BaseFN_CharacterFunction_PreFX}
  n6 = GetMaterialAttributes(None=@n5.Result) {attributeGetTypes [
    #1
    #2
    #3
    #4
   ]}
  n7 = TextureCoordinate()
  n8 = ComponentMask(None=@n7) {g=true}
  n9 = Constant() {r=-0.225}
  n10 = Saturate(None=Divide(A=Add(A=@n8, B=@n9)) {constB=0.125})
  n11 = Multiply(A=Time(), B=ScalarParameter() {defaultValue=2, group=Sirens, parameterName=SirenSpeed}) {constB=1.5}
  n12 = OneMinus(None=Frac(None=Add(A=@n10, B=@n11)))
  n13 = Multiply(A=Add(A=Sine(None=@n12))) {constB=0.5}
  n14 = MaterialFunctionCall!(Base=Multiply(A=@n6.Emissive Color) {constB=0.2}, Add=Multiply(A=@n13) {constB=4}) {materialFunction=/Engine/Functions/Engine_MaterialFunctions02/Math/MultiplyAdd}
  n15 = VectorParameter() {parameterName=SocketLocationR, defaultValue=<rgba 2810 1336 2075 0>}
  n16 = LinearInterpolate(A=@n1.RGB, B=@n15.RGB) {constAlpha=-0.15}
  n17 = LinearInterpolate(Alpha=@n13) {constB=8}
  n18 = OneMinus(None=Fresnel(Normal=MaterialFunctionCall!(Sphere Position=@n16, Sphere Radius=@n17, RayDirection, World position) {materialFunction=/Engine/Functions/Engine_MaterialFunctions02/RayTracedSphere}.Intersection Surface Normal) {baseReflectFraction=0, exponent=0.5})
  n19 = LinearInterpolate(A=@n1.RGB, B=@n15.RGB) {constAlpha=0.175}
  n20 = OneMinus(None=Fresnel(Normal=MaterialFunctionCall(Sphere Position=@n19, Sphere Radius=@n17) {materialFunction=&RayTracedSphere}.Intersection Surface Normal) {baseReflectFraction=0, exponent=0.5})
  n21 = LinearInterpolate(A=@n1.RGB, B=@n15.RGB) {constAlpha=0.825}
  n22 = OneMinus(None=Fresnel(Normal=MaterialFunctionCall(Sphere Position=@n21, Sphere Radius=@n17) {materialFunction=&RayTracedSphere}.Intersection Surface Normal) {baseReflectFraction=0, exponent=0.5})
  n23 = LinearInterpolate(A=@n1.RGB, B=@n15.RGB) {constAlpha=1.15}
  n24 = OneMinus(None=Fresnel(Normal=MaterialFunctionCall(Sphere Position=@n23, Sphere Radius=@n17) {materialFunction=&RayTracedSphere}.Intersection Surface Normal) {baseReflectFraction=0, exponent=0.5})
  n25 = Saturate(None=Add(A=Add(A=@n18, B=@n20), B=Add(A=@n22, B=@n24)))
  n26 = Multiply(A=@n6.Emissive Color, B=Multiply(A=@n13, B=Power(Base=@n25) {constExponent=1.5})) {constB=20}
  n27 = Power(Base=@n25) {constExponent=7}
  n28 = LinearInterpolate(
   A=ScalarParameter() {defaultValue=0.5, group=Sirens, parameterName=Siren_minParallax}
   B=ScalarParameter() {defaultValue=0.54, group=Sirens, parameterName=Siren_maxParallax}
   Alpha=@n4
  ) {constA=0.5, constB=0.54}
  n29 = LinearInterpolate(Alpha=@n13) {constB=0.025}
  n30 = Constant()
  n31 = SphereMask(A=@n7, B=AppendVector(A=@n28, B=Constant() {r=0.24}), Radius=@n29, Hardness=@n30) {attenuationRadius=0.03, hardnessPercent=50}
  n32 = SphereMask(A=@n7, B=AppendVector(A=@n28, B=Constant() {r=0.27}), Radius=@n29, Hardness=@n30) {attenuationRadius=0.03, hardnessPercent=50}
  n33 = SphereMask(A=@n7, B=AppendVector(A=@n28, B=Constant() {r=0.31}), Radius=@n29, Hardness=@n30) {attenuationRadius=0.03, hardnessPercent=50}
  n34 = SphereMask(A=@n7, B=AppendVector(A=@n28, B=Constant() {r=0.34}), Radius=@n29, Hardness=@n30) {attenuationRadius=0.03, hardnessPercent=50}
  n35 = Saturate(None=Add(A=Add(A=@n31, B=@n32), B=Add(A=@n33, B=@n34)))
  n36 = Multiply(A=@n6.Emissive Color, B=Multiply(A=@n13, B=Power(Base=@n35) {constExponent=1.5})) {constB=20}
  n37 = Power(Base=@n35) {constExponent=7}
  n38 = StaticSwitchParameter(
   True=Add(A=Multiply(A=@n26) {constB=2}, B=Multiply(A=@n27) {constB=3})
   False=Add(A=Multiply(A=@n36) {constB=2}, B=Multiply(A=@n37) {constB=3})
  ) {group=Sirens, parameterName=rayTraceSphereSirens}
  MaterialFunctionCall(
   AO=@n5.AO
   Distance Blend=@n5.AO
   Fresnel=@n5.AO
   Gmap=@n5.AO
   Material=SetMaterialAttributes!(MaterialAttributes=@n6.MaterialAttributes, Emissive Color=Add(A=@n14.Result, B=@n38) {constB=0}, Base Color=@n6.MaterialAttributes) {attributeSetTypes [#1, #4]}
  ) {materialFunction=&BaseFN_CharacterFunction_PostFX}
  ObjectOrientation()
  SphereMask(A=ComponentMask(None=@n7) {r=true}, B=Constant() {r=0.52}) {attenuationRadius=0.03, hardnessPercent=0}
 ]
}
